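# One-shot stereo MetaSound that plays UmbraMolt encounter music as an enter stinger, combat loop and defeat stinger.
# A Harmonix metronome quantizes both stingers; the combat loop starts as the enter stinger nearly finishes, with a gain ramp, and fades out under the defeat stinger.
# Pausing fades the music, and a weapon-fire envelope follower modulates its output gain.
# The sound finishes after the defeat stinger or a completed pause fade.
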MetaSoundSource /Umbramolt_Hole/Assets/Audio/Music/MSS_UmbraMolt_Music
inputs [
 MusicEncounter.bIsPaused {type=Bool}
 MusicEncounter.PauseFadeOutTime {type=Time}
 OnDefeated {type=Trigger}
 Tempo {type=Float}
 UE.Source.OnPlay {type=Trigger}
]
interfaces [UE.OutputFormat.Stereo:1.0, UE.Source.OneShot:1.0, UE.Source:1.1]
outputs [
 "UE.OutputFormat.Stereo.Audio:0" {type=Audio}
 "UE.OutputFormat.Stereo.Audio:1" {type=Audio}
 UE.Source.OneShot.OnFinished {type=Trigger}
]
pages [Default {
  input_defaults {MusicEncounter.PauseFadeOutTime=3, MusicEncounter.bIsPaused=null, OnDefeated=null, Tempo=95, UE.Source.OnPlay=null}
  nodes [
   DecibelstoLinearGain_1 = UE.Decibels to Linear Gain:Float() {also {defaults {Decibels=6}}}
   n1 = HarmonixNodes.Metronome(Tempo=^Tempo, Transport=HarmonixNodes.TriggerToTransport(Play=^UE.Source.OnPlay).Transport) {also {defaults {"Loop Length (Bars)"=1, Time Sig. Numerator=1}}}
   WavePlayer_2 = UE.Wave Player:Stereo(Play=HarmonixNodes.MidiQuantizeTriggerNode(MIDI Clock=@n1.MIDI Clock, Trigger Input=UE.Reroute(Value=^UE.Source.OnPlay).Value) {also {version [0,1]}}.Trigger Out) {also {defaults {Wave Asset=&/MX_S39_IP_UmbraMolt_Enter_STNG}, version [1,1]}}
   WavePlayer_0 = UE.Wave Player:Stereo(Play=@WavePlayer_2.On Nearly Finished) {also {
     defaults {Loop="<MetaSoundBuilderResult.SUCCEEDED: 0>", Wave Asset=&/MX_S39_IP_UmbraMolt_Combat_LP}
     version [1,1]
    }}
   AudioMixerAudioMixerStereo2_0 = "AudioMixer.Audio Mixer (Stereo, 2)"(
    Gain 1=HarmonixNodes.Ramper(Start=@DecibelstoLinearGain_1.Linear Gain, Snap To Start=@WavePlayer_2.On Nearly Finished) {also {defaults {Ramp Time=2, Target=1}, version [0,1]}}.Value
    In 1 L=@WavePlayer_0.Out Left
    In 1 R=@WavePlayer_0.Out Right
    In 0 L=@WavePlayer_2.Out Left
    In 0 R=@WavePlayer_2.Out Right
   )
   n2 = VariableMutator.MIDIClock(Value=@n1.MIDI Clock, Variable=InitVariable.MIDIClock().Variable)
   WavePlayer_1 = UE.Wave Player:Stereo(Play=HarmonixNodes.MidiQuantizeTriggerNode(Trigger Input=^OnDefeated, MIDI Clock=VariableAccessor.MIDIClock(Variable=@n2.Variable).Value) {also {version [0,1]}}.Trigger Out) {also {defaults {Wave Asset=&/MX_S39_IP_UmbraMolt_Defeat_STNG}, version [1,1]}}
   AudioMixerAudioMixerStereo2_1 = "AudioMixer.Audio Mixer (Stereo, 2)"(
    In 0 L=@AudioMixerAudioMixerStereo2_0.Out L
    In 0 R=@AudioMixerAudioMixerStereo2_0.Out R
    Gain 0=UE.InterpTo:Audio(Target=Value.Value:Float(Set=@WavePlayer_1.On Play) {also {defaults {Init Value=1, Target Value=0}}}.Output Value) {also {defaults {Interp Time=0.5}}}.Value
    In 1 L=@WavePlayer_1.Out Left
    In 1 R=@WavePlayer_1.Out Right
   )
   DecibelstoLinearGain_0 = UE.Decibels to Linear Gain:Float() {also {defaults {Decibels=-3}}}
   TriggerOnValueChange_0 = UE.Trigger On Value Change:Bool(Value=^MusicEncounter.bIsPaused)
   n3 = TriggerCompare.Clamp:Bool(A=^MusicEncounter.bIsPaused, Compare=@TriggerOnValueChange_0.Trigger) {also {defaults {B="<MetaSoundBuilderResult.SUCCEEDED: 0>"}}}
   n4 = Value.Value:Float(Reset=@n3.False, Set=@n3.True) {also {defaults {Init Value=1, Target Value=0}}}
   n5 = UE.Multiply:Float(
    AdditionalOperands=Modulation.GetModulatorValue() {also {defaults {Modulator=/Game/Sounds/Modulation/EnvelopeFollowers/FortWeapon/MP_WeaponFireEnvFollower}}}.Out
    PrimaryOperand=UE.Multiply:Float(
     AdditionalOperands=@DecibelstoLinearGain_0.Linear Gain
     PrimaryOperand=UE.InterpTo:Audio(Interp Time=^MusicEncounter.PauseFadeOutTime, Target=@n4.Output Value).Value
    ).Out
   )
   Multiply_0 = UE.Multiply:Audio by Float(PrimaryOperand=@AudioMixerAudioMixerStereo2_1.Out L, AdditionalOperands=@n5.Out)
   Multiply_1 = UE.Multiply:Audio by Float(PrimaryOperand=@AudioMixerAudioMixerStereo2_1.Out R, AdditionalOperands=@n5.Out)
   TriggerDelay_0 = UE.Trigger Delay(Delay Time=^MusicEncounter.PauseFadeOutTime, Reset=@n4.On Reset, In=@n4.On Set)
   TriggerAnyTriggerAny2_0 = "TriggerAny.Trigger Any (2)"(In 1=@TriggerDelay_0.Out, In 0=@WavePlayer_1.On Finished)
  ]
  links [
   {dst="@UE.OutputFormat.Stereo.Audio:0", src=Multiply_0.Out}
   {dst="@UE.OutputFormat.Stereo.Audio:1", src=Multiply_1.Out}
   {dst="@UE.Source.OneShot.OnFinished", src=TriggerAnyTriggerAny2_0.Out}
  ]
 }]
properties {
 assetImportData=&self:AssetImportData
 attenuationSettings=&/ATT_UmbraMolt_Music
 maxDistance=10000
 numChannels=2
 outputFormat=Stereo
 sampleRate=48000
 soundClassObject=&Fort_Music_NPC
 assetUserData [&self:AudioAssetUserData_0]
}
